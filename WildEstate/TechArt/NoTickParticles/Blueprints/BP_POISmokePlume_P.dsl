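; Smoke plume point-of-interest actor that configures a plume mesh with body and card material overrides plus a localized post-process effect.
; At construction it sets the post-process material and area transform, then packs box extent, actor axes, shading, fade and color into the post-process mesh's custom primitive data via a vector-to-floats helper.
; One frame after BeginPlay it enables custom depth on the static mesh; Tick and overlap events are empty.
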(event EventBeginPlay
  (|CallParentFunction)
  (Utilities|FlowControl|DelayUntilNextFrame)
  (Rendering|SetRenderCustomDepth (Variables|Debug|GetStaticMeshComponent) true))

(event Collision|EventActorBeginOverlap
  (params (OtherActor :type Actor%20Object%20Reference)))

(event EventTick
  (params (DeltaSeconds :type Float%20%28single-precision%29)))

(fn SetFloatCPDsFromVector
  (params (Static%20Mesh :type Static%20Mesh%20Component%20Object%20Reference) (DataIndex :type Integer) (Value :type Vector))
  (bind (_value_x _value_y _value_z) (Math|Vector|BreakVector (Variables|GetValue_X)))
  (bind _value_x_2 (Math|Vector|BreakVector (Variables|GetValue_X)))
  (bind (_value_x_3 _value_y_2) (Math|Vector|BreakVector (Variables|GetValue_X)))
  (Rendering|Material|SetDefaultCustomPrimitiveDataFloat (Variables|GetStatic%20Mesh) (Variables|GetDataIndex) _value_x_2)
  (Rendering|Material|SetDefaultCustomPrimitiveDataFloat (Variables|GetStatic%20Mesh) (+ (Variables|GetDataIndex) 1) _value_y_2)
  (Rendering|Material|SetDefaultCustomPrimitiveDataFloat (Variables|GetStatic%20Mesh) (+ (Variables|GetDataIndex) 2) _value_z))

(fn SetParameters
  (bind _postprocessarea (Variables|Default|GetPostProcessArea))
  (bind _postprocessmesh (Variables|Default|GetPostProcessMesh))
  (Rendering|Material|SetMaterial _postprocessarea 0 (Variables|Parameters|GetPostProcessMaterial))
  (Transformation|SetRelativeLocation _postprocessarea (Variables|Parameters|GetPostProcessOffset) false false)
  (Transformation|SetRelativeScale3D _postprocessarea (Variables|Parameters|GetPostProcessScale))
  (CallFunction|SetFloatCPDsfromVector _postprocessmesh 0 (Components|Box|GetScaledBoxExtent (Variables|Default|GetPostProcessArea)))
  (CallFunction|SetFloatCPDsfromVector _postprocessmesh 3 (Math|Vector|GetForwardVector (Transformation|GetActorRotation)))
  (CallFunction|SetFloatCPDsfromVector _postprocessmesh 6 (Math|Vector|GetRightVector (Transformation|GetActorRotation)))
  (CallFunction|SetFloatCPDsfromVector _postprocessmesh 9 (Math|Vector|GetUpVector (Transformation|GetActorRotation)))
  (CallFunction|SetFloatCPDsfromVector _postprocessmesh 12 (* (Variables|Parameters|GetShadingScale) "0.700000,0.700000,0.700000"))
  (CallFunction|SetFloatCPDsfromVector _postprocessmesh 15 (* (Variables|Parameters|GetShadingOffset) "0.010000,0.010000,0.010000"))
  (Rendering|Material|SetDefaultCustomPrimitiveDataFloat _postprocessmesh 18 (Variables|Parameters|GetPostProcessFade))
  (CallFunction|SetFloatCPDsfromVector _postprocessmesh 19 (Math|Conversions|ToVector(LinearColor) (Variables|Parameters|GetColorMultiplier))))

(fn ConstructionScript
  (|CallParentFunction)
  (Default|SetParameters)
  (Rendering|Material|SetMaterial (Variables|Debug|GetStaticMeshComponent) 0 (Variables|Default|GetBodyMaterialOverride))
  (Rendering|Material|SetMaterial (Variables|Debug|GetStaticMeshComponent) 1 (Variables|Default|GetCardMaterialOverride)))
