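; BP_SwiftApple_Smoke is an actor Blueprint in the NoTickParticles tech-art folder, apparently a smoke particle effect.
; On BeginPlay it waits one frame, then enables Render CustomDepth on its SM_ParticleStack31 component.
; The Tick and overlap events and the construction script are empty, so the graph holds no per-frame logic.
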(event EventBeginPlay
  (Utilities|FlowControl|DelayUntilNextFrame)
  (Rendering|SetRenderCustomDepth (Variables|Default|GetSM_ParticleStack31) true))

(event Collision|EventActorBeginOverlap
  (params (OtherActor :type Actor%20Object%20Reference)))

(event EventTick
  (params (DeltaSeconds :type Float%20%28single-precision%29)))

(fn ConstructionScript)
