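# Material function that turns a scalar gradient, such as a signed distance field, into an anti-aliased mask at a Cutoff threshold.
# It divides Cutoff minus Gradient by the length of its DDX/DDY screen derivatives, converting the difference to a distance in pixels.
# Subtracting that from one half and saturating gives a roughly one-pixel-wide edge at any scale, white where Gradient exceeds Cutoff.
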MaterialFunction /WildEstate/TechArt/NoTickParticles/SDF/AntiAliasing
graph {nodes [
  n1 = FunctionInput() {bUsePreviewValueAsDefault=true, inputName=Cutoff, inputType=FunctionInput_Scalar, previewValue=<0.5 0 0 1>}
  n2 = Subtract(
   A=@n1
   B=FunctionInput(Preview=MaterialFunctionCall() {materialFunction=&LinearGradient}.UGradient) {bUsePreviewValueAsDefault=true, inputName=Gradient, inputType=FunctionInput_Scalar}
  )
  n3 = DDX(Value=@n2)
  n4 = DDY(Value=@n2)
  n5 = Divide(A=@n2, B=Length(None=AppendVector(A=@n3, B=@n4)) {is=class MaterialExpressionLength})
  FunctionOutput(None=Saturate(None=Subtract(B=@n5) {constA=0.5}))
 ]}
