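# SDF_Circle is a material function that computes a signed distance field for a centered circle, with filled and stroked masks.
# It remaps UV around the center and subtracts a radius from its length; the radius is a default-midpoint lerp between Size minus and plus stroke thickness.
# Absolute distance minus thickness, optionally relative to Size, gives the stroke SDF.
# EdgeMin and EdgeMax turn both distances into saturated masks, output alongside the raw SDFs.
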MaterialFunction /WildEstate/TechArt/NoTickParticles/SDF/SDF_Circle
graph {nodes [
  n1 = Multiply(A=FunctionInput(Preview=TextureCoordinate()) {bUsePreviewValueAsDefault=true, inputName=UV, inputType=FunctionInput_Vector2}) {constB=2}
  n2 = FunctionInput() {bUsePreviewValueAsDefault=true, inputName=Size, inputType=FunctionInput_Scalar, sortPriority=1, previewValue=<0.5 0.5 0.5 0.5>}
  n3 = FunctionInput() {bUsePreviewValueAsDefault=true, inputName=Stroke Thickness, inputType=FunctionInput_Scalar, sortPriority=4, previewValue=<0.001 0 0 1>}
  n4 = FunctionInput(Preview=StaticBool() {value=true}) {
   bUsePreviewValueAsDefault=true
   inputName=Stroke Thickness Relative
   inputType=FunctionInput_StaticBool
   sortPriority=5
   previewValue=<0.1 0 0 1>
  }
  n5 = StaticSwitch(True=Multiply(A=@n2, B=@n3), False=@n3, Value=@n4)
  n6 = Subtract(
   A=Length(None=Subtract(A=@n1)) {is=class MaterialExpressionLength, desc=distance from cur pos to center}
   B=LinearInterpolate(A=Subtract(A=@n2, B=@n5), B=Add(A=@n2, B=@n5))
  )
  FunctionOutput(None=@n6) {outputName=SDF, sortPriority=2}
  n7 = FunctionInput() {bUsePreviewValueAsDefault=true, inputName=EdgeMin, inputType=FunctionInput_Scalar, sortPriority=2, previewValue=<0 0 0 0>}
  n8 = FunctionInput() {bUsePreviewValueAsDefault=true, inputName=EdgeMax, inputType=FunctionInput_Scalar, sortPriority=3, previewValue=<0.01 0 0 0.01>}
  n9 = InverseLinearInterpolate(A=@n7, B=@n8, Value=@n6)
  FunctionOutput(None=Saturate(None=OneMinus(None=@n9))) {outputName=Fill}
  n10 = Subtract(A=Abs(None=@n6), B=@n5)
  FunctionOutput(None=@n10) {outputName=SDF Stroke, sortPriority=3}
  n11 = InverseLinearInterpolate(A=@n7, B=@n8, Value=@n10)
  FunctionOutput(None=Saturate(None=OneMinus(None=@n11))) {outputName=Stroke, sortPriority=1}
 ]}
