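# Material function that draws a cloud card shape for the NoTickParticles effects, outputting an Outline and a Mask.
# A static switch, on by default, reads both from the GasCloud mask texture or builds them procedurally from six offset SDF circles.
# The circles' fills and strokes merge by max and are antialiased, with fill subtracted from stroke to leave only the outer outline.
# Camera distance varies stroke thickness, and CardOutlineOpacity fades the outline toward white.
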MaterialFunction /WildEstate/TechArt/NoTickParticles/MF_WildEstate_DynCloudCard
graph {nodes [
  Constant() {r=0.5}
  n1 = TextureSample(UVs=NamedRerouteUsage()) {samplerType=SAMPLERTYPE_Masks, texture=/WildEstate/Assets/FX/GasCloud/GasCloud_M}
  n2 = NamedRerouteUsage()
  n3 = MaterialFunctionCall!(
   UV=Add(A=NamedRerouteUsage(), B=Constant2Vector() {g=0.15, r=-0.06})
   Size=Constant() {r=0.54}
   EdgeMin
   EdgeMax
   Stroke Thickness=@n2
   Stroke Thickness Relative
  ) {materialFunction=/WildEstate/TechArt/NoTickParticles/SDF/SDF_Circle}
  n4 = MaterialFunctionCall(UV=Add(A=NamedRerouteUsage(), B=Constant2Vector() {g=0.045, r=0.2}), Size=Constant() {r=0.5}, Stroke Thickness=@n2) {materialFunction=&SDF_Circle}
  n5 = MaterialFunctionCall(UV=Add(A=NamedRerouteUsage(), B=Constant2Vector() {g=-0.203, r=0.2}), Size=Constant() {r=0.44}, Stroke Thickness=@n2) {materialFunction=&SDF_Circle}
  n6 = MaterialFunctionCall(UV=Add(A=NamedRerouteUsage(), B=Constant2Vector() {g=-0.203, r=-0.048}), Size=Constant() {r=0.47}, Stroke Thickness=@n2) {materialFunction=&SDF_Circle}
  n7 = Max(A=Max(A=Max(A=@n3.Stroke, B=@n4.Stroke), B=@n5.Stroke), B=@n6.Stroke)
  n8 = MaterialFunctionCall(UV=Add(A=NamedRerouteUsage(), B=Constant2Vector() {g=-0.14, r=-0.136}), Size=Constant() {r=0.47}, Stroke Thickness=@n2) {materialFunction=&SDF_Circle}
  n9 = MaterialFunctionCall(UV=Add(A=NamedRerouteUsage(), B=Constant2Vector() {g=0.08, r=-0.21}), Size=Constant() {r=0.5}, Stroke Thickness=@n2) {materialFunction=&SDF_Circle}
  n10 = MaterialFunctionCall!(Cutoff=Constant() {r=0.1}, Gradient=Max(A=Max(A=@n7, B=@n8.Stroke), B=@n9.Stroke)) {materialFunction=/WildEstate/TechArt/NoTickParticles/SDF/AntiAliasing}
  n11 = Max(A=Max(A=Max(A=@n3.Fill, B=@n4.Fill), B=@n5.Fill), B=@n6.Fill)
  n12 = MaterialFunctionCall(Cutoff=Constant() {r=0.9}, Gradient=Max(A=Max(A=@n11, B=@n8.Fill), B=@n9.Fill)) {materialFunction=&AntiAliasing}
  n13 = OneMinus(None=Saturate(None=Subtract(A=@n10.Result, B=@n12.Result)))
  n14 = NamedRerouteUsage()
  FunctionOutput(None=LinearInterpolate(B=StaticSwitch(True=@n1.G, False=@n13, Value=@n14), Alpha=ScalarParameter() {defaultValue=0.1, parameterName=CardOutlineOpacity}) {constA=1}) {outputName=Outline}
  n15 = Max(A=@n12.Result, B=@n10.Result)
  FunctionOutput(None=StaticSwitch(True=@n1.R, False=Saturate(None=@n15), Value=NamedRerouteUsage())) {outputName=Mask, sortPriority=1}
  n16 = CameraPositionWS()
  n17 = WorldPosition()
  n18 = Divide(A=Subtract(A=Distance(A=@n16, B=@n17.XYZ))) {constB=2000}
  NamedRerouteDeclaration(None=LinearInterpolate(
    A=ScalarParameter() {defaultValue=0.01, parameterName=FarThickness}
    B=ScalarParameter() {defaultValue=0.03, parameterName=NearThickness}
    Alpha=Saturate(None=@n18)
   )) {name=LineThicness, nodeColor=<rgba 0.6641 1 0 1>}
  NamedRerouteDeclaration(None=FunctionInput(Preview=TextureCoordinate()) {bUsePreviewValueAsDefault=true, inputName=UVs, inputType=FunctionInput_Vector2}) {name=UVs, nodeColor=<rgba 1 0 0.9844 1>}
  NamedRerouteDeclaration(None=FunctionInput(Preview=StaticBool() {value=true}) {bUsePreviewValueAsDefault=true, inputName=UseStaticTexture, inputType=FunctionInput_StaticBool, sortPriority=1}) {name=UseTexture, nodeColor=<rgba 0 1 0.6953 1>}
 ]}
