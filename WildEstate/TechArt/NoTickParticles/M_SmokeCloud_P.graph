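# Masked unlit stylized smoke/gas cloud material for meshes, with an IsBody switch between inset body and puff cards.
# WPO expands cards into sprites scaled by view angle, camera and pivot distance and vertex color (random, thickness, face mask), adding sine wobble.
# Color combines quantized sunlight shading, Fresnel glow and time-of-day tint.
# Cards use rotating textures and height-based pixel depth offset; noise breaks up near-camera fading.
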Material /WildEstate/TechArt/NoTickParticles/M_SmokeCloud_P
graph {
 properties {
  bAutomaticallySetUsageInEditor=false
  bHasPixelAnimation=true
  bUsedWithMeshDeformer=true
  bUsedWithSkeletalMesh=true
  blendMode=BLEND_Masked
  shadingModel=MSM_Unlit
  usedShadingModels=Unlit
 }
 nodes [
  n1 = VertexNormalWS()
  n2 = Multiply(
   A=ScalarParameter() {defaultValue=-1, group=WPO, parameterName=BodyInset}
   B=MaterialFunctionCall() {materialFunction=&ObjectScale}.Scale Z
  )
  n3 = ScalarParameter() {defaultValue=14.0368, group=WPO, parameterName=Sprite_Scale_Sides}
  n4 = VertexColor()
  n5 = Add(
   A=LinearInterpolate(A=ScalarParameter() {defaultValue=6.202, group=WPO, parameterName=Sprite_Scale_Thin}, B=@n3, Alpha=NamedRerouteUsage())
   B=Multiply(A=ScalarParameter() {defaultValue=3.9214, group=WPO, parameterName=Sprite_Scale_Random}, B=Subtract(A=@n4.R) {constB=0.5})
  )
  n6 = ScalarParameter() {defaultValue=29.4715, group=WPO, parameterName=Sprite_Scale_Back}
  n7 = LinearInterpolate(A=ScalarParameter() {defaultValue=-0.9246, group=WPO, parameterName=Sprite_Scale_Back_Thin}, B=@n6, Alpha=NamedRerouteUsage())
  n8 = VertexNormalWS()
  n9 = WorldPosition() {worldPositionShaderOffset=WPT_CameraRelativeNoOffsets}
  n10 = NamedRerouteDeclaration(None=Normalize(VectorInput=@n9.XYZ)) {name=CameraVector_Custom, nodeColor=<rgba 0 1 0.8828 1>}
  n11 = Saturate(None=NamedRerouteDeclaration(None=DotProduct(A=@n8, B=@n10)) {name=DotSurface, nodeColor=<rgba 0.2422 0 1 1>})
  n12 = ScalarParameter() {group=WPO, parameterName=Distance_Offset}
  n13 = MaterialFunctionCall() {materialFunction=&ObjectScale}
  n14 = Divide(
   A=Subtract(A=Length(None=@n9.XYZ) {is=class MaterialExpressionLength}, B=Multiply(A=@n12, B=@n13.Scale Z))
   B=Multiply(A=ScalarParameter() {defaultValue=29.5774, group=WPO, parameterName=Distance_Blend}, B=@n13.Scale Z)
  )
  n15 = Multiply(
   A=LinearInterpolate(
    A=ScalarParameter() {defaultValue=4, group=WPO, parameterName=ScaleNear}
    B=LinearInterpolate(A=@n5, B=@n7, Alpha=@n11)
    Alpha=Saturate(None=@n14)
   )
   B=MaterialFunctionCall() {materialFunction=&ObjectScale}.Scale Z
  )
  n16 = Distance(A=CameraPositionWS(), B=WorldPosition().XYZ)
  n17 = ScalarParameter() {defaultValue=-3000, parameterName=Shrink.Offset}
  n18 = Saturate(None=Divide(A=Add(A=@n16, B=@n17), B=ScalarParameter() {defaultValue=500, parameterName=Shrink.Distance}))
  n19 = ActorPositionWS()
  n20 = VectorParameter() {group=Scale, parameterName=PivotScale.CustomLocation, defaultValue=<rgba 0 0 0 1>}
  n21 = ScalarParameter() {group=Scale, parameterName=PivotScale.UseCustomPivotLocation}
  n22 = Distance(
   A=StaticComponentMaskParameter(None=LinearInterpolate(A=@n19, B=@n20.RGB, Alpha=@n21)) {defaultB=true, defaultG=true, defaultR=true, group=Scale, parameterName=PivotScale.AxisMask}
   B=StaticComponentMaskParameter(None=WorldPosition().XYZ) {defaultB=true, defaultG=true, defaultR=true, group=Scale, parameterName=PivotScale.AxisMask}
  )
  n23 = ScalarParameter() {defaultValue=-3000, group=Scale, parameterName=PivotScale.CenterDistanceOffset}
  n24 = Saturate(None=Divide(A=Add(A=@n22, B=@n23), B=ScalarParameter() {defaultValue=500, group=Scale, parameterName=PivotScale.Distance}))
  n25 = LinearInterpolate(
   A=ScalarParameter() {defaultValue=1, group=Scale, parameterName=PivotScale.MinScale}
   Alpha=LinearInterpolate(A=@n24, B=OneMinus(None=@n24), Alpha=ScalarParameter() {group=Scale, parameterName=PivotScale.Inverse})
  )
  n26 = Add(A=VertexColor().B, B=ScalarParameter() {parameterName=VertColorBlueScalarOffset})
  n27 = Multiply(A=Multiply(A=Multiply(A=@n15, B=@n18), B=@n25), B=LinearInterpolate(Alpha=Saturate(None=@n26)) {constA=0.05})
  n28 = Multiply(
   A=MaterialFunctionCall!(Center Location=WorldPosition().XYZ, 0-1 UVs, XY scale=@n27, OverridePivot, "Normals (Optional)") {materialFunction=&Sprite}.World position offset
   B=ShadowReplace(Default=ScalarParameter() {defaultValue=1, group=WPO, parameterName=SpriteMasterScale}, Shadow=Constant())
  )
  n29 = VertexNormalWS()
  n30 = NamedRerouteUsage()
  n31 = Multiply(
   A=LinearInterpolate(
    A=ScalarParameter() {defaultValue=-2.2442, group=WPO, parameterName=SurfaceOffset_thin}
    B=ScalarParameter() {defaultValue=-4.9573, group=WPO, parameterName=SurfaceOffset}
    Alpha=NamedRerouteUsage()
   )
   B=OneMinus(None=Saturate(None=@n30))
  )
  n32 = NamedRerouteUsage()
  n33 = NamedRerouteUsage()
  n34 = Multiply(
   A=Saturate(None=Multiply(A=@n33) {constB=-1})
   B=LinearInterpolate(
    A=ScalarParameter() {defaultValue=-4.8221, group=WPO, parameterName=PushBackFront_thin}
    B=ScalarParameter() {defaultValue=-4.3507, group=WPO, parameterName=PushBackFront}
    Alpha=NamedRerouteUsage()
   )
  )
  n35 = Multiply(
   A=Add(A=Multiply(A=@n29, B=@n31), B=Multiply(A=@n32, B=@n34))
   B=MaterialFunctionCall() {materialFunction=&ObjectScale}.Scale Z
  )
  n36 = NamedRerouteUsage()
  n37 = VertexColor()
  Add(
   A=StaticSwitchParameter(True=Multiply(A=@n1, B=@n2), False=Add(A=@n28, B=@n35)) {parameterName=IsBody}
   B=LinearInterpolate(A=@n36, B=LinearInterpolate(B=@n36, Alpha=@n37.B), Alpha=ScalarParameter() {group=WPO, parameterName=RemoveSineWobbleBlue})
  )
  AtmosphericLightVector()
  n38 = Fresnel(ExponentIn=ScalarParameter() {defaultValue=2.3911, group=Color, parameterName=Fresnel}) {baseReflectFraction=0}
  n39 = VectorParameter() {group=TODM, parameterName=ColorTint_Morning, defaultValue=<rgba 1.2 0.8375 0.9343 1>}
  n40 = VectorParameter() {group=TODM, parameterName=ColorTint_Day, defaultValue=<rgba 0.8333 0.8333 0.8333 1>}
  n41 = VectorParameter() {group=TODM, parameterName=ColorTint_Evening, defaultValue=<rgba 1.3 0.984 0.9141 1>}
  n42 = VectorParameter() {group=TODM, parameterName=ColorTint_Night, defaultValue=<rgba 0.3034 0.7123 1 1>}
  n43 = ScalarParameter() {defaultValue=0.5849, group=Color, parameterName=Brightness}
  n44 = NamedRerouteUsage()
  n45 = CollectionParameter() {
   collection=&FortniteMaterialParameters
   parameterName=SunlightDirectionalVector
  }
  n46 = Add(A=Add(A=NamedRerouteUsage(), B=DotProduct(A=@n44, B=@n45)))
  n47 = ScalarParameter() {defaultValue=2, group=Color, parameterName=ShadinSteps}
  n48 = Round(None=Multiply(A=Divide(A=@n46), B=@n47))
  n49 = LinearInterpolate(
   A=VectorParameter() {group=Color, parameterName=Color2, defaultValue=<rgba 0.5325 0.9797 0.1837 1>}.RGB
   B=VectorParameter() {group=Color, parameterName=Color, defaultValue=<rgba 0.7631 2 0.2917 1>}.RGB
   Alpha=Max(A=Divide(A=@n48, B=@n47), B=ScalarParameter() {defaultValue=0.3016, group=Color, parameterName=MinDarkness})
  )
  n50 = Add(
   A=Multiply(A=VectorParameter() {group=Color, parameterName=Glow, defaultValue=<rgba 0.0097 0.5 0 1>}.RGB, B=@n38)
   B=Multiply(
    A=MaterialFunctionCall(Color_Morning=@n39.RGB, Color_Day=@n40.RGB, Color_Evening=@n41.RGB, Color_Night=@n42.RGB) {materialFunction=&Color_Time_of_Day}.Result
    B=Multiply(A=@n43, B=@n49)
   )
  )
  n51 = NamedRerouteUsage()
  n52 = Saturate(None=LinearInterpolate(Alpha=@n51) {constA=-1})
  n53 = Multiply(A=NamedRerouteUsage(), B=StaticSwitchParameter(True=@n50, False=Multiply(A=@n50, B=@n52)) {parameterName=IsBody})
  n54 = Power(Base=@n53, Exp=ScalarParameter() {defaultValue=3, group=Color, parameterName=GlowContrast})
  n55 = ScalarParameter() {defaultValue=0.1, group=Color, parameterName=Glow}
  EyeAdaptationInverse(LightValueInput=Add(A=Multiply(A=@n54, B=@n55), B=@n53))
  LinearInterpolate(B=@n51, Alpha=NamedRerouteUsage()) {constA=1}
  Multiply(A=@n51) {constB=10}
  Multiply(A=@n25, B=@n6)
  Multiply(A=@n3, B=@n25)
  n56 = Constant3Vector() {constant=<rgba 1 0 0 0>}
  n57 = MaterialFunctionCall(
   Plane Normal Axis=CameraVectorWS()
   Plane X Axis=Normalize(VectorInput=Transform(None=@n56) {transformSourceType=TRANSFORMSOURCE_Camera})
   Plane Center=ObjectPositionWS()
   UV Scale=Multiply(
    A=ScalarParameter() {defaultValue=76.3289, group=Noise, parameterName=Noise_Size, sortPriority=36}
    B=MaterialFunctionCall() {materialFunction=&ObjectScale}.Scale Z
   )
  ) {materialFunction=&VirtualPlaneCoordinates}
  NamedRerouteDeclaration(None=TextureSample(UVs=@n57.UVs) {samplerType=SAMPLERTYPE_Masks, texture=/WildEstate/Assets/FX/GasCloud/GasCloud_Tiling_M}.RGB) {name=NoiseTexture, nodeColor=<rgba 0.4297 1 0 1>}
  n58 = WorldPosition()
  n59 = CameraPositionWS()
  n60 = Distance(A=@n58.XYZ, B=@n59)
  n61 = ScalarParameter() {defaultValue=1594.8198, parameterName=CameraFade_Offset}
  n62 = Saturate(None=Divide(A=Subtract(A=@n60, B=@n61), B=ScalarParameter() {defaultValue=953.3473, parameterName=CameraFade_Distance}))
  n63 = MaterialFunctionCall(Transition Phase=@n62, Height Texture=ComponentMask(None=NamedRerouteUsage()) {r=true}) {materialFunction=&HeightLerp}
  n64 = Subtract(A=@n63.Alpha, B=ScalarParameter() {defaultValue=-0.6584, parameterName=IntersectionOutline})
  n65 = ComponentMask(None=NamedRerouteUsage()) {g=true}
  n66 = Floor(None=@n62)
  NamedRerouteDeclaration(None=Multiply(A=Floor(None=@n64), B=LinearInterpolate(A=@n65, Alpha=@n66) {constA=1})) {name=OutlinesIntersection, nodeColor=<rgba 0.0547 0 1 1>}
  NamedRerouteDeclaration(None=@n66) {name=NearMask, nodeColor=<rgba 1 0.1875 0 1>}
  n67 = TextureSample(UVs=@n57.UVs) {samplerType=SAMPLERTYPE_Grayscale, texture=&LowResBlurredNoise}
  NamedRerouteDeclaration(None=@n67.R) {name=NoiseTexture_Gray, nodeColor=<rgba 0.4297 1 0 1>}
  NamedRerouteDeclaration(None=@n18) {name=ShrinkAlpha, nodeColor=<rgba 0.3828 0 1 1>}
  NamedRerouteDeclaration(None=@n27) {name=Scale, nodeColor=<rgba 0.9688 1 0 1>}
  n68 = WorldPosition()
  n69 = MaterialFunctionCall(Float2=Subtract(A=TextureCoordinate()) {constB=0.5}) {materialFunction=&BreakOutFloat2Components}
  n70 = CollectionParameter() {
   collection=&FortniteMaterialParameters
   parameterName=SunlightDirectionalVector
  }
  n71 = ComponentMask(None=@n70) {b=true, g=true, r=true}
  n72 = MaterialFunctionCall(Float2=NamedRerouteUsage()) {materialFunction=&BreakOutFloat2Components}
  n73 = Transform(None=Constant3Vector() {constant=<rgba 0 -1 0 0>}) {transformSourceType=TRANSFORMSOURCE_Camera}
  n74 = Add(A=Multiply(A=Multiply(A=@n69.R, B=@n71), B=@n72.R), B=Multiply(A=Multiply(A=@n69.G, B=@n73), B=@n72.G))
  NamedRerouteDeclaration(None=Add(
    A=Subtract(A=@n74, B=@n68.XYZ)
    B=MaterialFunctionCall() {materialFunction=&ComponentPivotLocation}.Result
   ) {desc=WPO}) {name=ShadowWPO, nodeColor=<rgba 0 0.6484 1 1>}
  n75 = OneMinus(None=TextureSample() {samplerType=SAMPLERTYPE_LinearGrayscale, texture=/WildEstate/Assets/FX/GasCloud/GasCloud_height}.R)
  n76 = Multiply(
   A=ScalarParameter() {defaultValue=3, group=Pixel Depth Offset, parameterName=PDO}
   B=MaterialFunctionCall() {materialFunction=&ObjectScale}.Scale Z
  )
  NamedRerouteDeclaration(None=StaticSwitchParameter(True=Constant(), False=Multiply(A=@n75, B=@n76)) {parameterName=IsBody}) {name=PDO, nodeColor=<rgba 1 0 0.2812 1>}
  NamedRerouteDeclaration(None=Power(Base=VertexColor().G, Exp=ScalarParameter() {defaultValue=0.6469, group=WPO, parameterName=Thickness_Falloff})) {name=ThicknessMask, nodeColor=<rgba 1 0.2812 0 1>}
  n77 = VertexNormalWS()
  n78 = WorldPosition()
  n79 = MaterialFunctionCall() {materialFunction=&ObjectScale}
  n80 = ScalarParameter() {defaultValue=45, group=Normal, parameterName=OffsetNormal}
  n81 = VertexColor() {desc="Red is Random\r\nGreen Is Thickness\r\nBlue is Face Mask"}
  n82 = Subtract(
   A=WorldPosition() {worldPositionShaderOffset=WPT_ExcludeAllShaderOffsets}.XYZ
   B=Multiply(A=Multiply(A=@n79.Scale Z, B=@n80), B=Multiply(A=@n81.G, B=@n77))
  )
  n83 = Normalize(VectorInput=LinearInterpolate(A=@n77, B=Subtract(A=@n78.XYZ, B=@n82), Alpha=VertexColor() {desc="Red is Random\r\nGreen Is Thickness\r\nBlue is Face Mask"}.B))
  NamedRerouteDeclaration(None=StaticSwitchParameter(True=@n77, False=StaticSwitchParameter(True=@n77, False=@n83) {parameterName=UseVertexNormalsForLighting}) {parameterName=IsBody}) {name=Normals, nodeColor=<rgba 1 0.7031 0 1>}
  n84 = Time()
  n85 = VertexColor()
  n86 = Multiply(A=ScalarParameter() {defaultValue=1, group=Animation, parameterName=Rotation_Speed}, B=Sign(None=Subtract(A=@n85.R) {constB=0.5}))
  n87 = ScalarParameter() {defaultValue=44.0961, group=Animation, parameterName=RandomUVRot}
  n88 = Rotator(Time=Add(A=Multiply(A=@n84, B=@n86), B=Multiply(A=@n85.R, B=@n87)))
  n89 = MaterialFunctionCall!(UVs=@n88, UseStaticTexture) {materialFunction=&/MF_WildEstate_DynCloudCard}
  NamedRerouteDeclaration(None=@n89.Outline) {name=TextureOutline, nodeColor=<rgba 0.3828 0 1 1>}
  NamedRerouteDeclaration(None=@n89.Mask) {name=TextureMask, nodeColor=<rgba 0 1 0.0625 1>}
  n90 = ScalarParameter() {defaultValue=0.3, group=Animation, parameterName=SineMovement_Amp}
  n91 = MaterialFunctionCall() {materialFunction=&ObjectScale}
  n92 = ScalarParameter() {defaultValue=11.1265, group=Animation, parameterName=SinSIze}
  n93 = Add(
   A=Divide(A=WorldPosition() {worldPositionShaderOffset=WPT_ExcludeAllShaderOffsets}.XYZ, B=Multiply(A=@n92, B=@n91.Scale Z))
   B=Multiply(A=ScalarParameter() {defaultValue=0.5, group=Animation, parameterName=SineMovement_Speed}, B=Time())
  )
  NamedRerouteDeclaration(None=Multiply(A=Multiply(A=@n90, B=@n91.Scale Z), B=Sine(None=@n93))) {name=SineWave Wobble, nodeColor=<rgba 0.5 1 0 1>}
  NamedRerouteDeclaration(None=@n63.Alpha) {name=NearCamera Mask, nodeColor=<rgba 0.0547 0 1 1>}
  NamedRerouteDeclaration(None=Multiply(A=ScalarParameter() {defaultValue=0.156, group=Noise, parameterName=Noise_Add}, B=Subtract(A=@n67.R) {constB=0.5})) {name=Noise, nodeColor=<rgba 0.2656 1 0 1>}
  NamedRerouteUsage()
  NamedRerouteUsage()
  NamedRerouteUsage()
  OneMinus(None=@n38)
  n94 = SmoothStep(Value=NamedRerouteUsage()) {constMax=0.1}
  OneMinus(None=@n94)
  PixelDepth()
  SphereMask(A=@n58.XYZ, B=@n59)
  n95 = WorldPosition()
  n96 = CameraPositionWS()
  n97 = Subtract(A=MaterialFunctionCall(
    TextureObject=TextureObject() {samplerType=SAMPLERTYPE_Grayscale, texture=&LowResBlurredNoise}
    TextureSize=ScalarParameter() {defaultValue=512, parameterName=NearCameraCull.NoiseSize}
   ) {materialFunction=&WorldAlignedTexture}.XYZ Texture) {constB=0.5}
  n98 = ScalarParameter() {defaultValue=1, parameterName=NearCameraCull.NoiseIntensity}
  n99 = Subtract(
   A=Add(A=Distance(A=@n95.XYZ, B=@n96), B=Multiply(A=@n97, B=@n98))
   B=ScalarParameter() {defaultValue=1594.8198, parameterName=NearCameraCull.Offset}
  )
  n100 = ScalarParameter() {defaultValue=953.3473, parameterName=NearCameraCull.Distance}
  StaticSwitchParameter(True=Saturate(None=Divide(A=@n99, B=@n100)), False=Multiply(A=NamedRerouteUsage(), B=@n94)) {parameterName=IsBody}
  TextureSample(UVs=@n88) {samplerType=SAMPLERTYPE_Masks, texture=/WildEstate/Assets/FX/GasCloud/GasCloud_M}
 ]
}
