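# Material function that shades GlowDonut smoke with a stepped light mask and a time-of-day tint, outputting processed color and mask.
# With particle data enabled, bounding-box shading from particle spawn position plus an offset is quantized into steps and floored at a minimum.
# That mask blends darkened and full color before tinting by morning, day, evening and night colors.
# Sun-direction and camera-plane scrolling noise branches exist but do not reach either output.
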MaterialFunction /WildEstate/TechArt/GlowDonut/MaterialFunctions/MF_Smoke_Shading
graph {nodes [
  Bounds()
  n1 = CollectionParameter() {
   collection=&FortniteMaterialParameters
   parameterName=SunlightDirectionalVector
  }
  n2 = Add(A=NamedRerouteUsage(), B=DotProduct(A=VertexNormalWS(), B=@n1))
  n3 = FunctionInput() {bUsePreviewValueAsDefault=true, inputName=ParticleSpawnPos, sortPriority=1, previewValue=<1 1 1 1>}
  n4 = StaticSwitch(
   True=Add(
    A=MaterialFunctionCall!(PartcileSpawnPos=@n3) {materialFunction=/WildEstate/TechArt/MaterialFunctions/NoTickParticles/MF_NoTickParticles_BoundingBoxShading}.Mask
    B=ScalarParameter() {group=Shading, parameterName=BBX.ShadingOffset}
   )
   False=Constant()
   Value=FunctionInput(Preview=StaticBool()) {bUsePreviewValueAsDefault=true, inputName=UseParticleData, inputType=FunctionInput_StaticBool, sortPriority=2}
  )
  Divide(A=Add(A=Add(A=@n2, B=@n4)))
  n5 = VectorParameter() {group=Shading, parameterName=TOD.Color.Morning, defaultValue=<rgba 1 0.8854 0.8977 1>}
  n6 = VectorParameter() {group=Shading, parameterName=TOD.Color.Day, defaultValue=<rgba 1 1 1 1>}
  n7 = VectorParameter() {group=Shading, parameterName=TOD.Color.Evening, defaultValue=<rgba 1 0.8598 0.8229 1>}
  n8 = VectorParameter() {group=Shading, parameterName=TOD.Color.Night, defaultValue=<rgba 0.7604 0.8102 1 1>}
  n9 = NamedRerouteUsage()
  n10 = Multiply(B=@n9) {constA=0.6}
  n11 = ScalarParameter() {defaultValue=3, group=Shading, parameterName=ShadingSteps}
  n12 = Multiply(A=@n4, B=@n11)
  n13 = Max(A=Divide(A=Round(None=@n12), B=@n11), B=ScalarParameter() {defaultValue=0.4, group=Shading, parameterName=ShadingDarkenMin}) {constB=0.3}
  FunctionOutput(None=Multiply(
    A=MaterialFunctionCall(Color_Morning=@n5.RGB, Color_Day=@n6.RGB, Color_Evening=@n7.RGB, Color_Night=@n8.RGB) {materialFunction=&Color_Time_of_Day}.Result
    B=LinearInterpolate(A=@n10, B=@n9, Alpha=@n13)
   )) {outputName=Processed Color}
  FunctionOutput(None=Saturate(None=@n13)) {outputName=Mask, sortPriority=2}
  n14 = ScalarParameter() {defaultValue=76.3289, group=Shading, parameterName=Noise_Size, sortPriority=36}
  Multiply(A=@n14, B=MaterialFunctionCall() {materialFunction=&ObjectScale}.Scale Z)
  n15 = Transform(None=Constant3Vector() {constant=<rgba 1 0 0 0>}) {transformSourceType=TRANSFORMSOURCE_Camera}
  Multiply(A=@n15, B=Constant3Vector() {constant=<rgba 1 0 1 0>})
  NamedRerouteDeclaration(None=FunctionInput() {inputName=Color, previewValue=<0.6 0.6 0.68 1>}) {name=Color, nodeColor=<rgba 0.4531 0 1 1>}
  n16 = ActorPositionWS()
  n17 = VectorParameter() {group=Shading, parameterName=Shading.NoiseVP.CenterOffset, defaultValue=<rgba 0 0 0 1>}
  n18 = Constant()
  n19 = Time()
  n20 = Add(
   A=MaterialFunctionCall(Plane Normal Axis=CameraVectorWS(), Plane X Axis=Normalize(VectorInput=@n15), Plane Center=Add(A=@n16, B=@n17.RGB), UV Scale=@n14) {materialFunction=&VirtualPlaneCoordinates}.UVs
   B=Multiply(A=AppendVector(A=@n18, B=@n19), B=ScalarParameter() {defaultValue=-0.1, group=Shading, parameterName=ShadingNoiseSpeed})
  )
  NamedRerouteDeclaration(None=TextureSample(UVs=@n20) {samplerType=SAMPLERTYPE_Masks, texture=/WildEstate/Assets/FX/GasCloud/GasCloud_Tiling_M}.RGB) {name=NoiseTexture, nodeColor=<rgba 0.4297 1 0 1>}
  n21 = TextureSample(UVs=@n20) {samplerType=SAMPLERTYPE_LinearGrayscale, texture=/WildEstate/Assets/FX/GasCloud/GasCloud_Tiling_Gray}
  NamedRerouteDeclaration(None=@n21.R) {name=NoiseTexture_Gray, nodeColor=<rgba 0.4297 1 0 1>}
  NamedRerouteDeclaration(None=Multiply(A=ScalarParameter() {defaultValue=0.156, group=Shading, parameterName=Noise_Add}, B=Subtract(A=@n21.R) {constB=0.5})) {name=Noise, nodeColor=<rgba 0.2656 1 0 1>}
  ObjectPositionWS()
 ]}
